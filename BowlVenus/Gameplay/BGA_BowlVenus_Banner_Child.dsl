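; BowlVenus child of a banner Blueprint that animates a flag's capture progress, overriding the parent to omit player banners.
; BeginPlay creates base and captured flag materials on authority and initializes the progress display on remotes.
; UpdateFlagProgress lowers the banner ring, then raises it, scrubbing a timeline by progress to drive material dissolve.
; Remotes also swap mesh and material at a threshold, play swap audio and feed the hologram parameter collection.
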(event EventBeginPlay
  (bind _bannerring (Variables|Default|GetBannerRing))
  (bind (_relativelocation_x _relativelocation_y _relativelocation_z) (Math|Vector|BreakVector (Variables|/Script/Engine.SceneComponent|GetRelativeLocation_X (Variables|Default|GetPoleBottom))))
  (bind (_relativelocation_x_2 _relativelocation_y_2 _relativelocation_z_2) (Math|Vector|BreakVector (Variables|/Script/Engine.SceneComponent|GetRelativeLocation_X (Variables|Default|GetBannerRing))))
  ;[0.04081 0.9323 0.004537 0.5] Override from parent as we don't want the player banners
  (Variables|Default|SetbUsingBaseFlag true)
  (Networking|SwitchHasAuthority
    (:Authority
      (Variables|Default|SetBannerRingInitialZ _relativelocation_z_2)
      (Variables|Default|SetPole%20Bottom%20Relative%20Location%20Z _relativelocation_z)
      (bind _returnvalue (Rendering|Material|CreateDynamicMaterialInstance _bannerring 0 "/BowlVenus/Assets/Materials/MI_BowlVenus_Flagpole_B.MI_BowlVenus_Flagpole_B" "None" :!class /Script/Engine.PrimitiveComponent))
      (Variables|Default|SetCaptured_Flag _returnvalue)
      (bind _returnvalue_2 (Rendering|Material|CreateDynamicMaterialInstance _bannerring 0 "/BowlVenus/Assets/Materials/MI_BowlVenus_Flagpole_A.MI_BowlVenus_Flagpole_A" "None" :!class /Script/Engine.PrimitiveComponent))
      (bind _output_get (Variables|Default|SetBase_Flag _returnvalue_2))
      (Rendering|Material|SetMaterial _bannerring 0 _output_get))
    (:Remote
      (CallFunction|InitializeProgressWidget)
      (CallFunction|UpdateFlagProgress))))

(event Collision|EventActorBeginOverlap
  (params (OtherActor :type Actor%20Object%20Reference))
  (|CallParentFunction OtherActor) ; This node is disabled and will not be called.\nDrag off pins to build functionality.
)

(event EventTick
  (params (DeltaSeconds :type Float%20%28single-precision%29))
  (|CallParentFunction DeltaSeconds) ; This node is disabled and will not be called.\nDrag off pins to build functionality.
)

(event Custom|UpdateFlagDissolve
  (bind (_direction _dissolve_percentage) (Timeline|Flag_Material_Timeline SetNewTime :NewTime (Variables|Default|GetFlagProgress))
    (:Update
      (if (> _dissolve_percentage 0)
        (Rendering|Material|SetScalarParameterValueonMaterials (Variables|Default|GetBannerRing) "FlagDissolveFX_Percentage" _dissolve_percentage)))))

(fn UpdateFlagMeshMaterial
  (bind _bannerring (Variables|Default|GetBannerRing))
  (if (!= (< (Variables|Default|GetFlagProgress) (Variables|Default|GetFlagSwapPercent)) (Variables|Default|GetbUsingBaseFlag))
    (if (< (Variables|Default|GetFlagProgress) (Variables|Default|GetFlagSwapPercent))
      ;[1 1 1] Base Flag
      (LocalVariables|SetNewFlagMaterial (Variables|Default|GetBase_Flag))
      (LocalVariables|SetNewFlagMesh (Variables|Default|GetBaseFlagMesh))
      (Variables|Default|SetbUsingBaseFlag true)
      (else
        ;[1 1 1] Captured Flag
        (LocalVariables|SetNewFlagMaterial (Variables|Default|GetCaptured_Flag))
        (LocalVariables|SetNewFlagMesh (Variables|Default|GetCapturedFlagMesh))
        (Variables|Default|SetbUsingBaseFlag false)))
    (Components|StaticMesh|SetStaticMesh _bannerring (LocalVariables|GetNewFlagMesh))
    (Rendering|Material|SetMaterial _bannerring 0 (LocalVariables|GetNewFlagMaterial))))

(fn UpdateFlagProgress
  (bind (_relativelocation_x _relativelocation_y _relativelocation_z) (Math|Vector|BreakVector (Variables|/Script/Engine.SceneComponent|GetRelativeLocation_X (Variables|Default|GetPoleTop))))
  (bind _flagprogress (Variables|Default|GetFlagProgress))
  (bind (_relativelocation_x_2 _relativelocation_y_2 _relativelocation_z_2) (Math|Vector|BreakVector (Variables|/Script/Engine.SceneComponent|GetRelativeLocation_X (Variables|Default|GetPoleBottom))))
  ;[0 0.9271 0.03368 0.5] For this flag progress, we want to:
  ; 1. Lower from Top to bottom
  ; 2. Swap flag material
  ; 3. Raise back to Top
  ; Make sure this doesnt' look terrible when another team does their own progress
  (Utilities|FlowControl|Sequence
    (:then_0
      ;;[1 1 1] Check flag swap. Play sound and swap mesh/material.
      (Networking|SwitchHasAuthority
        (:Remote
          (if (Variables|Default|GetbPlaySwapSound)
            (Variables|Default|SetbPlaySwapSound false)
            (Audio|PlaySoundatLocation (Variables|Default|GetFlag%20Swap%20Sound) (Transformation|GetWorldLocation (Variables|Default|GetBannerRing))))
          (CallFunction|UpdateFlagMeshMaterial))))
    (:then_1
      (Transformation|SetRelativeLocation (Variables|Default|GetBannerRing) (Math|Vector|MakeVector 0.0 0.0 (Math|Float|SelectFloat (Math|Float|Lerp (Variables|Default|GetPole%20Bottom%20Relative%20Location%20Z) _relativelocation_z (Math|Float|MapRangeClamped (Variables|Default|GetFlagProgress) 0.5 1.0 0.0 1.0)) (Math|Float|Lerp (Variables|Default|GetBannerRingInitialZ) _relativelocation_z_2 (Math|Float|MapRangeClamped (Variables|Default|GetFlagProgress) 0.0 0.5 0.0 1.0)) (> (Variables|Default|GetFlagProgress) 0.5))) false false)
      ;;[1 1 1] moved to /Script/Engine.Blueprint'/POIBanner/Actors/CapturePoint/BGA_POIBanner_CapturePoint.BGA_POIBanner_CapturePoint' to fix possible slow replication
      (Networking|SwitchHasAuthority
        (:Remote
          (Rendering|Material|SetScalarParameterValue "/Game/Environments/TechArt/Hologram/MPC_Hologram.MPC_Hologram" "ProgressNormalized" (Variables|Default|GetFlagProgress) :!class /Script/Engine.KismetMaterialLibrary)
          (if (> _flagprogress 0)
            (Rendering|Material|SetScalarParameterValue "/Game/Environments/TechArt/Hologram/MPC_Hologram.MPC_Hologram" "ScreensEnabled" 1.0 :!class /Script/Engine.KismetMaterialLibrary)
            (elif (<= _flagprogress 0)
              (Rendering|Material|SetScalarParameterValue "/Game/Environments/TechArt/Hologram/MPC_Hologram.MPC_Hologram" "ScreensEnabled" :!class /Script/Engine.KismetMaterialLibrary))))))
    (:then_2
      (CallFunction|UpdateFlagDissolve))))

(fn ConstructionScript
  (|CallParentFunction))
