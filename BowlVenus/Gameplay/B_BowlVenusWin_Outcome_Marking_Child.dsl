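; Child Blueprint of a marking outcome, apparently for a BowlVenus win, that customizes the effect played when marking happens.
; On PlayMarkingFX it offsets the referenced POI banner's pole-top location by the loot spawn offset.
; Where cosmetic events can execute, it moves and activates a Niagara effect there and plays the capture pulse scan sound.
; BeginPlay, overlap and Tick add no logic.
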(event EventBeginPlay
  (|CallParentFunction) ; This node is disabled and will not be called.\nDrag off pins to build functionality.
)

(event Collision|EventActorBeginOverlap
  (params (OtherActor :type Actor%20Object%20Reference)))

(event EventTick
  (params (DeltaSeconds :type Float%20%28single-precision%29)))

(event EventPlayMarkingFX
  (bind _bannerref (Variables|Default|GetBannerRef))
  (bind _poletop (Variables|/POIBanner/Actors/BGA_POIBanner_Banner.BGA_POIBanner_Banner_C|GetPoleTop _bannerref))
  (bind _returnvalue (Transformation|GetWorldLocation _poletop))
  (bind _lootspawnoffset (Variables|Default|GetLootSpawnoffset))
  (bind _returnvalue_2 (+ _returnvalue _lootspawnoffset))
  (bind _niagara (Variables|Default|GetNiagara))
  (Networking|CanExecuteCosmeticEvents
    (:True
      (Transformation|SetWorldLocation _niagara _returnvalue_2 false false)
      (Components|Activation|Activate _niagara false)
      (Audio|PlaySoundatLocation "/POIBanner/Sounds/POI_Captured_Pulse_Marking_Scan_Cue.POI_Captured_Pulse_Marking_Scan_Cue" _returnvalue_2))))

(fn ConstructionScript
  (|CallParentFunction))
